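; Parent Blueprint for ToyFox ability projectiles that explode on stopping or when an optional hotfixable fuse timer halts their movement.
; The server runs gated async pawn and world targeting, applying damage GameplayEffects with a per-tick target cap.
; Buildings the blast would kill skip the effect and go to the GiantPawn destruction manager for direct destruction.
; Exploded or disabled projectiles are hidden, and homing stops once the target is passed.
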(event EventBeginPlay
  (bind _hotfix_fuseduration (Variables|Hotfix|GetHotfix_FuseDuration))
  (bind _returnvalue (ScalableFloat|GetValueatLevel _hotfix_fuseduration))
  ;[0 0.05208 0.00876 0.5] BeginPlay: Start fuse timer.
  (|CallParentFunction)
  (Networking|SwitchHasAuthority
    (:Authority
      (if (> _returnvalue 0)
        (bind _returnvalue_2 (Utilities|Time|SetTimerbyEvent (EventDispatchers|CreateEvent :!event HandleFuseTimerComplete) _returnvalue))
        (Variables|Transient|SetFuseTimerHandle _returnvalue_2)))))

(event EventOnStop
  (params (Hit :type Hit%20Result%20Structure%20%28by%20ref%29))
  ;[0.01563 0.15 0.05241 0.5] OnStop: Explode the projectile when it stops on hit.
  (|CallParentFunction Hit)
  (Networking|SwitchHasAuthority
    (:Authority
      (Delivery|DoExplosion)
      (CallFunction|ExecuteExplosionDamage))))

(event EventOnExploded
  (params (HitActors :type Array%20of%20Actor%20Object%20References) (HitResults :type Array%20of%20Hit%20Result%20Structures))
  ;[0.1228 0.02285 0.15 0.5] OnExploded/Disabled: Hide the projectile.
  (|CallParentFunction HitActors HitResults)
  (Utilities|FlowControl|DoOnce false
    (:Completed
      (Default|CallOnProjectileExploded)
      (Networking|CanExecuteCosmeticEvents
        (:True
          (Rendering|SetActorHiddenInGame true))))))

(event EventOnDisableForKill
  ;[0.1228 0.02285 0.15 0.5] OnExploded/Disabled: Hide the projectile.
  (|CallParentFunction)
  (Utilities|FlowControl|DoOnce false
    (:Completed
      (Default|CallOnProjectileExploded)
      (Networking|CanExecuteCosmeticEvents
        (:True
          (Rendering|SetActorHiddenInGame true))))))

(event Custom|ExecuteExplosionDamage
  (bind _returnvalue (CallFunction|MakeEffectContext (Ability|GetAbilitySystemComponent (Game|GetInstigator))))
  (bind _returnvalue_2 (Game|GetInstigator))
  (bind _returnvalue_3 (Ability|GetAbilitySystemComponent _returnvalue_2))
  ;[0.0499 0.04225 0.15 0.5] ExecuteExplosionDamage: Execute async explosion damage.
  (if (Networking|IsServer)
    (bind _updated_variable (Utilities|FlowControl|DeltaGateWildcard (Variables|Transient|GetbExplosionDamageDealt) true)
      (:Success
        (if _updated_variable
          (Utilities|IsValid _returnvalue_2
            (:Is%20Valid
              (Utilities|IsValid _returnvalue_3
                (:Is%20Valid
                  (bind (_asynctaskproxy _targetinghandle) (Class|AsyncActionPerformTargeting|PerformTargetingAsyncAction 0 (Variables|Settings|GetExplosionTargetingPresetPawn) true)
                    (:then
                      (bind (_asynctaskproxy_2 _targetinghandle_2) (Class|AsyncActionPerformTargeting|PerformTargetingAsyncAction 0 (Variables|Settings|GetExplosionTargetingPresetWorld) true)
                        (:then
                          (Utilities|Time|ClearandInvalidateTimerbyHandle (Variables|Transient|GetFuseTimerHandle)))
                        (:Targeted
                          (bind _buildingstoapplyge (CallFunction|ProcessWorldDamageResults _targetinghandle_2))
                          (Ability|Tasks|ApplyGameplayEffectToHitResultsAsync _returnvalue_3 _buildingstoapplyge (Variables|Settings|GetExplosionWorldDamageGameplayEffect) 1 1 _returnvalue (Variables|Hotfix|GetHotfix_MaxTargetsToProcessPerTick)))))
                    (:Targeted
                      (Ability|Tasks|ApplyGameplayEffectToTargetingHandleAsync _returnvalue_3 _targetinghandle (Variables|Settings|GetExplosionPawnDamageGameplayEffect) 1 1 _returnvalue (Variables|Hotfix|GetHotfix_MaxTargetsToProcessPerTick)))))
                (:Is%20Not%20Valid
                  (Fortnite|FortShippingLog self "ExecuteExplosionDamage: ASC is null." true))))
            (:Is%20Not%20Valid
              (Fortnite|FortShippingLog self "ExecuteExplosionDamage: Instigator is null." true))))))))

(event ProjectileMovement|EventOnHomingPassedTarget
  ;[0.05156 0.1268 0.15 0.5] OnHomingPassedTarget: If the target has been passed, stop homing and straighten out velocity
  (Projectile|SetHomingStyle 0 0 "None" false :!class /Script/FortniteGame.FortProjectileMovementInterface))

(fn HandleFuseTimerComplete
  (Networking|SwitchHasAuthority
    (:Authority
      (Game|Components|ProjectileMovement|StopSimulating (Variables|Movement|GetProjectileMovementComponent) (Collision|MakeHitResult true false 0.0 0.0 (Transformation|GetActorLocation) (Transformation|GetActorLocation))) ; Invokes the explosion.
    )))

(fn MakeEffectContext
  (params (AbilitySystemComponent :type Ability%20System%20Component%20Object%20Reference))
  ;[0.007824 0 0.15 0.5] Add the projectile as the source actor.
  (Utilities|FlowControl|Sequence
    (:then_0
      (Utilities|IsValid AbilitySystemComponent
        (:Is%20Valid
          (bind _output_get (LocalVariables|SetEffectContext (GameplayEffects|MakeEffectContext AbilitySystemComponent)))
          (Ability|EffectContext|AddSourceObject _output_get self)
          (return _output_get))))
    (:then_1
      (return 0))))

(fn ProcessWorldDamageResults
  (params (TargetingHandle :type Targeting%20Request%20Handle%20Structure))
  (LocalVariables|SetTotalTargetingResults (TargetingSystem|TargetingResults|GetTargetingResults (GameInstanceSubsystems|GetTargetingSubsystem) TargetingHandle))
  (Utilities|FlowControl|Sequence
    (:then_0
      (for _array_element (LocalVariables|GetTotalTargetingResults)
        (bind _output_get (LocalVariables|SetCurrentResult _array_element))
        (bind _asbuilding_actor (Utilities|Casting|CastToBuildingActor (.HitActor (Collision|BreakHitResult _output_get))))
        (if (Variables|/Script/Engine.Actor|GetbCanBeDamaged _asbuilding_actor)
          (if (<= (Health|GetHealth _asbuilding_actor :!class /Script/FortniteGame.BuildingActor) (ScalableFloat|GetCachedScalableFloatValue (Variables|Settings|GetWorldExplosionDamage)))
            (Utilities|Array|Add (LocalVariables|GetBuildingsToDestroyDirectly) _asbuilding_actor)
            (else
              (Utilities|Array|Add (LocalVariables|GetBuildingsToApplyGE) (LocalVariables|GetCurrentResult)))))))
    (:then_1
      (Public|AddBuildingsforDestruction (Actor|GetComponentbyClass (Fortnite|Athena|GameState|GetGameStateAthena) "/GiantPawnCore/Gameplay/Destruction/B_GiantPawnDestructionManagerGameStateComponent.B_GiantPawnDestructionManagerGameStateComponent_C") (LocalVariables|GetBuildingsToDestroyDirectly) (Utilities|GetClassDisplayName (Utilities|GetClass self)) (Variables|Actor|GetInstigator))
      (return (LocalVariables|GetBuildingsToApplyGE)))))

(fn ConstructionScript)
