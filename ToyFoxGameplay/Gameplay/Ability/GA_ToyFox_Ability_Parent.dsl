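; GA_ToyFox_Ability_Parent is a parent Gameplay Ability Blueprint for ToyFox abilities that adds audio feedback for activation failures caused by cooldown.
; Its OverrideFailedReason override checks the failure tags for Activate.Failed.Cooldown and plays the ability-on-cooldown sound in 2D, intended for the local player.
; It then returns the failure reason unchanged.
; The activate and end events are empty.
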(event Ability|EventActivateAbility)

(event Ability|EventOnEndAbility
  (params (bWasCancelled :type Boolean)))

(fn OverrideFailedReason
  (params (FailedReason :type Gameplay%20Tag%20Container%20Structure%20%28by%20ref%29))
  ;[0.02249 0.15 0.1456 0.5] Overridden to play the sound of the ability being on cooldown for the local player.
  (Utilities|FlowControl|Sequence
    (:then_0
      (if (GameplayTags|HasTag FailedReason "(TagName=\"Activate.Failed.Cooldown\")")
        (Audio|PlaySound2D (Variables|Default|GetAbilityOnCooldownSound))))
    (:then_1
      (return FailedReason))))
