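; GA_ToyFox_ExitGiant is a gameplay ability that appears to run when the ToyFox player exits a giant state.
; On activation it waits three seconds, then casts the actor's player controller to FortPlayerController and clears its primary bar slotted items cache.
; It then ends the ability; the end-ability event is empty.
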(event Ability|EventActivateAbility
  (Utilities|FlowControl|Delay 3.0)
  (Utilities|FlowControl|Sequence
    (:then_0
      (bind _asfort_player_controller (Utilities|Casting|CastToFortPlayerController (Ability|GetPlayerControllerfromActorInfo)))
      (Class|FortPlayerController|ClearPrimaryBarSlottedItemsCache _asfort_player_controller))
    (:then_1
      (Ability|EndAbility))))

(event Ability|EventOnEndAbility
  (params (bWasCancelled :type Boolean)))
