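; GA_ToyFox_PrimaryFireAbility is the ToyFox primary fire ability, which chooses between a finisher and a melee attack.
; On activation by a valid player pawn, it runs a targeting task with the finisher detection preset.
; ParseTargetData returns the first building actor tagged as both ToyFox boss and actively stunned.
; A match sends the Finisher gameplay event with that target, otherwise the Melee event fires, and the ability ends.
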(event Ability|EventActivateAbility
  (CallFunction|GetAvatarasValidFortPlayerPawn
    (:ValidFortPlayerPawn
      (bind (_asynctaskproxy _targetdata) (Ability|Tasks|PerformTargetingAbilityTask "None" "/ToyFoxGameplay/Gameplay/Ability/Finisher/TP_ToyFox_Finisher_Detection.TP_ToyFox_Finisher_Detection" false false true)
        (:Targeted
          _)
        (:Cancelled
          _))
      (bind (_bactivatefinisher _finishertarget) (CallFunction|ParseTargetData _targetdata))
      (if _bactivatefinisher
        (Fortnite|FortShippingLog self "ActivateAbility: Firing Finisher Gameplay Event.")
        (Ability|SendGameplayEvent "(TagName=\"ToyFox.Ability.Finisher\")" (Utilities|Struct|MakeGameplayEventData "(TagName=\"\")" 0 0 _finishertarget))
        (else
          (Ability|SendGameplayEvent "(TagName=\"ToyFox.Ability.Melee\")")))
      (Ability|EndAbility))
    (:AvatarCastFailed
      (Ability|EndAbility))))

(event Ability|EventOnEndAbility
  (params (bWasCancelled :type Boolean)))

(fn ParseTargetData
  (params (TargetData :type Gameplay%20Ability%20Target%20Data%20Handle%20Structure%20%28by%20ref%29))
  (Utilities|FlowControl|Sequence
    (:then_0
      (for _array_element (Ability|TargetData|GetAllActorsfromTargetData TargetData)
        (bind _asbuilding_actor (Utilities|Casting|CastToBuildingActor _array_element))
        (if (GameplayTags|HasAllMatchingGameplayTags _asbuilding_actor (GameplayTags|MakeLiteralGameplayTagContainer "(GameplayTags=((TagName=\"TidalCrane.ToyFox.Boss\"),(TagName=\"Ability.TIdalCrane.ToyFox.Stunned.Active\")))"))
          (return true _array_element))))
    (:then_1
      (return false 0))))
